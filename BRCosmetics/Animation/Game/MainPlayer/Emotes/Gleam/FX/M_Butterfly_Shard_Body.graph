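# Unlit, two-sided translucent material for the butterfly shard body effect in the Gleam emote, usable on skeletal meshes.
# Bounding-box gradients mask panning smoke noise, which blends a dark blue base into bright cyan-blue cracks.
# A second, wider gradient mask combines with a camera-facing normal term, apparently for opacity.
# A pivot-relative vector, normalized by object radius and scaled in local space, appears to provide a vertex offset.
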Material /BRCosmetics/Animation/Game/MainPlayer/Emotes/Gleam/FX/M_Butterfly_Shard_Body
graph {
 properties {
  bUsedWithMeshDeformer=true
  bUsedWithSkeletalMesh=true
  blendMode=BLEND_Translucent
  pixelDepthOffsetMode=PDOM_Legacy
  shadingModel=MSM_Unlit
  twoSided=true
  usedShadingModels=Unlit
 }
 nodes [
  n1 = StaticComponentMaskParameter(None=MaterialFunctionCall() {materialFunction=&BoundingBoxBased_0-1_UVW}.RGB) {defaultB=true}
  n2 = SmoothStep(Value=@n1) {constMin=0.45}
  n3 = MaterialFunctionCall() {materialFunction=&BoundingBoxBased_0-1_UVW}
  n4 = AppendVector(A=@n3.R, B=@n3.R)
  n5 = AppendVector(A=Constant() {r=4}, B=Constant() {r=1})
  n6 = TextureSample(UVs=Panner(Coordinate=Multiply(A=@n4, B=@n5)) {speedY=0.5}) {texture=&T_Inky_Smoke_Tile_Chromatic}
  n7 = Multiply(A=Multiply(A=@n2) {constB=5}, B=LinearInterpolate(A=@n6.R, Alpha=@n2))
  n8 = Power(Base=@n6.R) {constExponent=5}
  LinearInterpolate(
   A=VectorParameter() {parameterName=BaseColor, defaultValue=<rgba 0.1137 0.1382 0.37 0>}.RGB
   B=VectorParameter() {parameterName=CrackColor, defaultValue=<rgba 1.0827 2.6624 5 0>}.RGB
   Alpha=Add(A=Saturate(None=@n7), B=Multiply(A=@n8) {constB=5})
  )
  n9 = DotProduct(A=CameraVectorWS(), B=VertexNormalWS())
  n10 = Power(Base=@n9)
  n11 = SmoothStep(Value=@n1) {constMax=0.8, constMin=0.3}
  n12 = Multiply(A=Multiply(A=@n11) {constB=5}, B=LinearInterpolate(A=@n6.R, Alpha=@n11))
  Multiply(A=Multiply(A=Saturate(None=@n10), B=Saturate(None=@n12))) {constB=2}
  Multiply(A=@n9) {constB=2}
  n13 = MaterialFunctionCall() {materialFunction=&ObjectPivotPoint}
  n14 = WorldPosition()
  n15 = Multiply(
   A=Divide(A=Subtract(A=@n13.Object Pivot Location, B=@n14.XYZ), B=ObjectRadius())
   B=MaterialFunctionCall() {materialFunction=&ObjectScale}.Scale XYZ
  )
  Transform(None=Multiply(A=Transform(None=@n15) {transformSourceType=TRANSFORMSOURCE_World, transformType=TRANSFORM_Local}, B=Constant() {r=27})) {transformSourceType=TRANSFORMSOURCE_Local}
 ]
}
