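# Niagara effect for the Gleam emote that surrounds the character with glowing, light-blue butterflies.
# Butterfly meshes spawn on a cylinder shell, drift under curl noise and point attraction, carry glow sprites and flickering lights, and emit location and death events.
# Sparks spawn from butterfly particles through an attribute reader, alongside brief evaporation flashes with lights.
# A refraction glow mesh expands at the system origin, its camera offset switched by a front-end preview flag.
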NiagaraSystem /BRCosmetics/Animation/Game/MainPlayer/Emotes/Gleam/FX/NS_GleamEmote_Butterflies
dependencies {
 usedDataInterfaces [&NiagaraDataInterfaceCurve, &NiagaraDataInterfaceParticleRead, &NiagaraDataInterfaceColorCurve]
 usedDynamicInputs []
 usedModules [&SystemState]
}
emitters {
 Butterfly {
  emitterSpawnScript [/SetVariables(Emitter.CameraFade=10) {isSetParameters=true}]
  emitterUpdateScript [
   &EmitterState()
   &SpawnRate(Spawn Probability=0.8, SpawnRate=3, Use Spawn Probability=true)
  ]
  particleSpawnScript [
   &InitializeParticle(
    Lifetime=&RandomRangeFloat(Minimum=2, Maximum=2.5)
    Mesh Scale Mode=::NewEnumerator3
    Mesh Uniform Scale=0.012
    Sprite Size Mode=::NewEnumerator3
    Uniform Sprite Size=15
    UsePositionOffset=false
    Write Material Random=true
   )
   &InitialMeshOrientation(Mesh Axis To Orient=<0 0 1>, Mesh Orientation Mode=::NewEnumerator2, Use Rotation Vector=false)
   &ShapeLocation(
    Cylinder Height=50
    Cylinder Radius=90
    MaskSpawnGroup=false
    Offset=<0 0 125>
    Offset Mode=::NewEnumerator0
    Override Seed=true
    Random Seed=7
    Randomness Mode:ENiagaraRandomnessMode=::NewEnumerator1
    Shape Primitive=::NewEnumerator1
    Surface Only=true
    Surface Only Band Thickness=1
    Use Endcaps In Surface Only Mode=false
    UseOwnerScale=true
   )
   &ApplyInitialForces()
   &AddVelocity(
    Module.Use Falloff Distance=false
    Origin Offset=<0 0 125>
    Override Seed=false
    Velocity Mode=::NewEnumerator1
    Velocity Speed=&RandomRangeFloat(Minimum=80, Maximum=100)
   )
   &GenerateLocationEvent(Use Event Probability=false)
  ]
  particleUpdateScript [
   &ParticleState()
   &DynamicMaterialParameters(
    Index 0 Param 1=0.25
    Index 0 Param 4=&Multiply_Float(A=$System.NormalizedLoopAge, B=10)
    Index 1 Param 1=3000
    Index 1 Param 2=1
    Module.Param0WriteEnabled=true
    Module.Param0WriteEnabled1=true
    Module.Param1WriteEnabled=true
    Module.Param1WriteEnabled1=true
    Module.Param2WriteEnabled=true
    Module.Param2WriteEnabled1=true
    Module.Param3WriteEnabled=true
    Module.Param3WriteEnabled1=true
    Write Parameter Index 1=true
   )
   &CurlNoiseForce(Noise Frequency=150, Noise Strength=250, Pan Noise=true)
   &PointAttractionForce(
    Attraction Radius=350
    Attractor Position Offset:Vector3f=<0 0 150>
    Kill Within Radius=false
    Module.Use Falloff=false
    UseAttractorOffset=true
   )
   &Drag(Use Linear Drag=true, Use Rotational Drag=false)
   &SolveForcesAndVelocity(Clamp Velocity=false, Limit Acceleration=false)
   &ScaleMeshSize(Scale Factor=&VectorFromFloat(Value=&FloatFromCurve(FloatCurve=DataInterface() {propertyValues {Curve {keys [
          <at 0 0 RCIM_Cubic>
          {arriveTangent=1, interpMode=RCIM_Cubic, leaveTangent=1, time=0.15, value=1}
          {interpMode=RCIM_Cubic, leaveTangent=-1.1765, time=1, value=1}
         ]}}})))
   &ScaleColor(ScaleA=true, ScaleRGB=true)
   /SetVariables(
    Particles.GlowColor=&MakeLinearColorFromVectorAndFloat(
     "Vector (RGB)"=&MakeVectorFromLinearColorRGB(Color=<rgba 0.3991 0.6626 1 1>)
     "Float (Alpha)"=&FloatFromCurve(
      FloatCurve=DataInterface() {propertyValues {Curve {keys [
          {arriveTangent=6.6736, interpMode=RCIM_Cubic, leaveTangent=6.6736, tangentMode=RCTM_User, time=0, value=0}
          <at 0.3 1>
          {interpMode=RCIM_Cubic, tangentMode=RCTM_User, time=0.65, value=1}
          {arriveTangent=-4.9432, interpMode=RCIM_Cubic, leaveTangent=-4.9432, tangentMode=RCTM_User, time=1, value=0}
         ]}}}
      Scale Curve=0.5
     )
    )
    Particles.GlowSize=&Multiply_Vector2D_ByFloat(Float=&MakeFloatFromVector(Vector=$Particles.Scale, Channel=::NewEnumerator0), Vector2D=<1250 1250>)
    Particles.SpriteDynamicParam:Vector4f=&MakeVector4(X=0.7, Y=1, Z=1, W=1)
   ) {isSetParameters=true}
   &GenerateDeathEvent(Use Event Probability=false)
   &FN_Gen5_Light(
    Attenutation=25
    Color:LinearColor=&MultiplyLinearColor(
     Linear Color=<rgba 0.5216 0.8706 1 1>
     Scale Factor=&MakeVector4_FromFloat(Float=&Multiply_Float(A=1, B=&Sine(Normalized Angle=$Emitter.Age, Period=0.1, Scale=0.1, Bias=1)))
    )
    Exponent=4
    Intensity=1
   )
  ]
  rendererClasses [
   &NiagaraSpriteRendererProperties
   &NiagaraMeshRendererProperties
   &NiagaraComponentRendererProperties
  ]
  renderers [
   {rendererClass=&NiagaraSpriteRendererProperties}
   {rendererClass=&NiagaraMeshRendererProperties, rendererIndex=1}
   {rendererClass=&NiagaraComponentRendererProperties, rendererIndex=2}
  ]
 }
 Butterfly_Sparks {
  emitterUpdateScript [
   &EmitterState(Life Cycle Mode=::NewEnumerator1, Loop Behavior=::NewEnumerator1, UseLoopDelay=false)
   &SpawnParticlesFromOtherEmitter(
    Attribute Reader=DataInterface() {propertyValues {EmitterBinding {bindingMode=Other, emitterName=Butterfly}}}
    Spawn Rate Per Particle Cap=50
    Use Maximum Age=false
    Use Minimum Age=false
   )
  ]
  particleSpawnScript [
   &InitializeParticle(
    AdjustAlpha=false
    AdjustHue=true
    AdjustSaturation=false
    AdjustValue=false
    Color=<rgba 0.2396 0.658 1 1>
    Color Mode=::NewEnumerator3
    Lifetime Max=1
    Lifetime Min=0.25
    Lifetime Mode=::NewEnumerator1
    Sprite Size Mode=::NewEnumerator1
    Uniform Sprite Size Max=6
    Uniform Sprite Size Min=2
    UsePositionOffset=false
    Write Material Random=true
   )
   &SampleParticlesFromOtherEmitter(
    Applied Position Offset:Vector3f=&RandomRangeVector(Minimum=<-4 -4 -4>, Maximum=<4 4 4>)
    Applied Velocity Scale=0.3
    Velocity Application:ENiagara_DirectReadApplicationMode=::NewEnumerator0
    Velocity Sampling=::NewEnumerator0
   )
   &ApplyInitialForces()
  ]
  particleUpdateScript [
   &ParticleState()
   &CurlNoiseForce(Noise Strength=100, Pan Noise=false)
   &Drag(Drag=3, Use Linear Drag=true, Use Rotational Drag=false)
   &GravityForce(Gravity=<0 0 -75>)
   &SolveForcesAndVelocity(Clamp Velocity=false, Limit Acceleration=false)
   &ScaleSpriteSize(Uniform Curve Sprite Scale=DataInterface() {propertyValues {Curve {keys [{interpMode=RCIM_Cubic, tangentMode=RCTM_User, time=0.3, value=1}, {interpMode=RCIM_Cubic, tangentMode=RCTM_User, time=1, value=0}]}}})
   &ScaleColor(ScaleA=true, ScaleRGB=false)
  ]
  rendererClasses [&NiagaraSpriteRendererProperties]
  renderers [{rendererClass=&NiagaraSpriteRendererProperties}]
 }
 Evaporation_Flash {
  emitterUpdateScript [&EmitterState(Life Cycle Mode=::NewEnumerator1, UseLoopDelay=false)]
  particleSpawnScript [&InitializeParticle(
    Color=<rgba 0.6458 0.8343 1 0.33>
    Lifetime=0.2
    Sprite Size Mode=::NewEnumerator3
    Uniform Sprite Size=60
    UsePositionOffset=false
    Write Material Random=true
   )]
  particleUpdateScript [
   &ParticleState()
   &ScaleSpriteSize(Uniform Curve Sprite Scale=DataInterface() {propertyValues {Curve {keys [
        {interpMode=RCIM_Cubic, tangentMode=RCTM_User, time=0, value=0}
        {interpMode=RCIM_Cubic, tangentMode=RCTM_User, time=0.5, value=1}
        {tangentMode=RCTM_User, time=1, value=0}
       ]}}})
   /SetVariables(Particles.WideFlareSize:Vector2f=&Multiply_Vector2D(
     A=$Particles.SpriteSize
     B=&Multiply_Vector2D_ByFloat(
      Float=&FloatFromCurve(FloatCurve=DataInterface() {propertyValues {Curve {keys [
           {arriveTangent=9.0725, interpMode=RCIM_Cubic, leaveTangent=9.0725, tangentMode=RCTM_User, time=0, value=0}
           {arriveTangent=-2.5249, interpMode=RCIM_Cubic, leaveTangent=-2.5249, tangentMode=RCTM_User, time=0.3, value=0.9847}
           {arriveTangent=-0.0405, interpMode=RCIM_Cubic, leaveTangent=-0.0405, tangentMode=RCTM_User, time=1, value=0}
          ]}}})
      Vector2D=<1.5 0.25>
     )
    )) {isSetParameters=true}
   &FN_Gen5_Light(
    Attenutation=50
    Color:LinearColor=<rgba 0.5216 0.8706 1 1>
    Exponent=4
    Intensity:NiagaraFloat=&FloatFromCurve(
     FloatCurve=DataInterface() {propertyValues {Curve {keys [
         {interpMode=RCIM_Cubic, tangentMode=RCTM_User, time=0.1, value=0}
         {interpMode=RCIM_Cubic, tangentMode=RCTM_User, time=0.3, value=1}
         {tangentMode=RCTM_User, time=0.8, value=0}
        ]}}}
     Scale Curve=5
    )
   )
  ]
  rendererClasses [&NiagaraSpriteRendererProperties, &NiagaraComponentRendererProperties]
  renderers [
   {rendererClass=&NiagaraSpriteRendererProperties}
   {rendererClass=&NiagaraSpriteRendererProperties, rendererIndex=1}
   {rendererClass=&NiagaraComponentRendererProperties, rendererIndex=2}
  ]
 }
 Refraction_Glow {
  emitterUpdateScript [&EmitterState(Life Cycle Mode=::NewEnumerator1, Loop Delay=0.1, UseLoopDelay=false)]
  particleSpawnScript [
   &InitializeParticle(Lifetime=0.3, Mesh Scale Mode=::NewEnumerator3, Mesh Uniform Scale=0.025, UsePositionOffset=false, Write Material Random=true) {module=InitializeParticle001}
   &InitialMeshOrientation(
    Rotation=&RandomRangeVector(Minimum=<0 0 0>, Maximum=<1 1 1>)
    Use Orientation Vector=true
    Use Rotation Vector=true
   ) {module=InitialMeshOrientation001}
   &SystemLocation(MaskSpawnGroup=false)
   &CameraOffset(Camera Offset Amount=&MakeCustomFloatFromBool(Bool=$User.bIsFrontEndPreview, True Float=1, False Float=30))
  ]
  particleUpdateScript [
   &ParticleState()
   &DynamicMaterialParameters(
    Index 0 Param 1=&FloatFromCurve(FloatCurve=DataInterface() {propertyValues {Curve {keys [<at 0 1.33>,<at 0.4 1>,<at 1 0.968>]}}})
    Index 0 Param 2=1
    Index 0 Param 3=1
    Index 0 Param 4=1
    Module.Param0WriteEnabled=true
    Module.Param1WriteEnabled=true
    Module.Param2WriteEnabled=true
    Module.Param3WriteEnabled=true
   )
   &ScaleColor(
    Scale Alpha=&FloatFromCurve(FloatCurve=DataInterface() {propertyValues {Curve {keys [
         {interpMode=RCIM_Cubic, tangentMode=RCTM_User, time=0, value=0}
         {interpMode=RCIM_Cubic, tangentMode=RCTM_User, time=0.2, value=1.05}
         {tangentMode=RCTM_User, time=1, value=0}
        ]}}})
    Scale RGB=&MakeVectorFromLinearColorRGB(Color=&ColorFromCurve(ColorCurve=DataInterface() {propertyValues {
        AlphaCurve {keys [<at 0 1 RCIM_Cubic>]}
        BlueCurve {keys [<at 0 1.029 RCIM_Cubic>,<at 0.6829 1 RCIM_Cubic>]}
        GreenCurve {keys [<at 0 2 RCIM_Cubic>,<at 0.6829 0.4176 RCIM_Cubic>]}
        RedCurve {keys [<at 0 0.9125 RCIM_Cubic>,<at 0.6829 0 RCIM_Cubic>]}
       }}))
    ScaleA=true
    ScaleRGB=true
   )
   &ScaleMeshSize(Scale Factor=&VectorFromFloat(Value=&FloatFromCurve(FloatCurve=DataInterface() {propertyValues {Curve {keys [
          {arriveTangent=4.1785, interpMode=RCIM_Cubic, leaveTangent=4.1785, tangentMode=RCTM_User, time=0, value=0}
          {arriveTangent=0.9995, interpMode=RCIM_Cubic, leaveTangent=0.9995, tangentMode=RCTM_User, time=0.2223, value=0.6825}
          {arriveTangent=-0.0405, interpMode=RCIM_Cubic, leaveTangent=-0.0405, tangentMode=RCTM_User, time=1, value=1}
         ]}}})))
  ]
  rendererClasses [&NiagaraMeshRendererProperties]
  renderers [{rendererClass=&NiagaraMeshRendererProperties, rendererIndex=0}]
 }
}
summary {
 systemName=NS_GleamEmote_Butterflies
 emitters [
  {
   emitterName=Butterfly
   rendererClasses [
    &NiagaraSpriteRendererProperties
    &NiagaraMeshRendererProperties
    &NiagaraComponentRendererProperties
   ]
  }
  {emitterName=Butterfly_Sparks, rendererClasses [&NiagaraSpriteRendererProperties]}
  {
   emitterName=Evaporation_Flash
   rendererClasses [&NiagaraSpriteRendererProperties, &NiagaraComponentRendererProperties]
  }
  {emitterName=Refraction_Glow, rendererClasses [&NiagaraMeshRendererProperties]}
 ]
 userVariables [User.bIsFrontEndPreview {type=NiagaraBool}]
}
