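# Niagara burst effect for the portal moment of the Gleam emote, built from five short-lived emitters.
# Evaporation_Flash pairs flare sprites with a component-rendered light, Fragments throws drag-slowed meshes outward, and Wisps adds a growing, fading sprite.
# Refraction_Glow and Wave are meshes animated by color, scale and dynamic material curves, with camera offset switched by User.bIsFrontEndPreview.
# Wave's alpha also scales with time of day.
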NiagaraSystem /BRCosmetics/Animation/Game/MainPlayer/Emotes/Gleam/FX/NS_GleamEmote_Portal
dependencies {
 usedDataInterfaces [&NiagaraDataInterfaceCurve, &NiagaraDataInterfaceColorCurve]
 usedDynamicInputs [&MakeVector4_FromLinearColor]
 usedModules [&SystemState]
}
emitters {
 Evaporation_Flash {
  emitterUpdateScript [
   &EmitterState()
   &SpawnBurst_Instantaneous(Spawn Time=0.05, Use Spawn Probability=false, UseLoopCountLimit=false)
  ]
  particleSpawnScript [&InitializeParticle(
    Color=<rgba 0.7744 1.3504 2 1>
    Lifetime=0.25
    Sprite Size Mode=::NewEnumerator3
    Uniform Sprite Size=70
    UsePositionOffset=false
    Write Material Random=true
   )]
  particleUpdateScript [
   &ParticleState()
   &ScaleSpriteSize(Uniform Curve Sprite Scale=DataInterface() {propertyValues {Curve {keys [
        {interpMode=RCIM_Cubic, tangentMode=RCTM_User, time=0, value=0}
        {interpMode=RCIM_Cubic, tangentMode=RCTM_User, time=0.5, value=1}
        {tangentMode=RCTM_User, time=1, value=0}
       ]}}})
   /SetVariables(Particles.WideFlareSize:Vector2f=&Multiply_Vector2D(
     A=$Particles.SpriteSize
     B=&Multiply_Vector2D_ByFloat(
      Float=&FloatFromCurve(FloatCurve=DataInterface() {propertyValues {Curve {keys [
           {arriveTangent=9.0725, interpMode=RCIM_Cubic, leaveTangent=9.0725, tangentMode=RCTM_User, time=0, value=0}
           {arriveTangent=-2.5249, interpMode=RCIM_Cubic, leaveTangent=-2.5249, tangentMode=RCTM_User, time=0.3, value=0.9847}
           {arriveTangent=-0.0405, interpMode=RCIM_Cubic, leaveTangent=-0.0405, tangentMode=RCTM_User, time=1, value=0}
          ]}}})
      Vector2D=<1.5 0.25>
     )
    )) {isSetParameters=true}
   &FN_Gen5_Light(
    Attenutation=100
    Color:LinearColor=$Particles.Color
    Exponent=4
    Intensity:NiagaraFloat=&FloatFromCurve(
     FloatCurve=DataInterface() {propertyValues {Curve {keys [
         {arriveTangent=7.1483, interpMode=RCIM_Cubic, leaveTangent=7.1483, tangentMode=RCTM_User, time=0.1, value=0}
         {interpMode=RCIM_Cubic, tangentMode=RCTM_User, time=0.2, value=1}
         {tangentMode=RCTM_User, time=0.7, value=0}
        ]}}}
     Scale Curve=5
    )
   )
  ]
  rendererClasses [&NiagaraSpriteRendererProperties, &NiagaraComponentRendererProperties]
  renderers [
   {rendererClass=&NiagaraSpriteRendererProperties}
   {rendererClass=&NiagaraSpriteRendererProperties, rendererIndex=1}
   {rendererClass=&NiagaraComponentRendererProperties, rendererIndex=2}
  ]
 }
 Fragments {
  emitterUpdateScript [
   &EmitterState(Life Cycle Mode=::NewEnumerator1, Loop Behavior=::NewEnumerator1, Loop Duration=1, UseLoopDelay=false)
   &SpawnBurst_Instantaneous(Spawn Count=6, Spawn Time=0.05, Use Spawn Probability=false)
  ]
  particleSpawnScript [
   &InitializeParticle(
    Color=<rgba 1.1458 3.6277 10 1>
    Lifetime Max=0.7
    Lifetime Min=0.5
    Lifetime Mode=::NewEnumerator1
    Mesh Scale Mode=::NewEnumerator1
    Mesh Uniform Scale Max=0.4
    Mesh Uniform Scale Min=0.3
    UsePositionOffset=false
    Write Material Random=true
   )
   &InitialMeshOrientation(
    Rotation=&RandomRangeVector(Minimum=<0 -1 0>, Maximum=<0 1 0>)
    Use Orientation Vector=true
    Use Rotation Vector=true
   )
   &AddVelocity(Module.Use Falloff Distance=false, Override Seed=false, Velocity Mode=::NewEnumerator1, Velocity Speed=150)
  ]
  particleUpdateScript [
   &Drag(Drag=5, Use Linear Drag=true, Use Rotational Drag=false)
   &GravityForce(Gravity=<0 0 -300>) {enabled=false}
   &SolveForcesAndVelocity(Clamp Velocity=false, Limit Acceleration=false)
   &ParticleState()
   &UpdateMeshOrientation(
    Delta Time:NiagaraFloat=$Particles.NormalizedAge
    Rotation Rate=&RandomRangeFloat(Minimum=-0.05, Maximum=0.05)
    Rotation Vector=<1 0 0>
   )
   &ScaleColor(
    Scale Alpha=&Multiply_Float(A=&FloatFromCurve(FloatCurve=DataInterface() {propertyValues {Curve {keys [<at 0 1>,<at 1 0>]}}}), B=1)
    ScaleA=true
    ScaleRGB=true
   )
   &ScaleMeshSize(Scale Factor=&VectorFromFloat(Value=&FloatFromCurve(FloatCurve=DataInterface() {propertyValues {Curve {keys [{interpMode=RCIM_Cubic, tangentMode=RCTM_User, time=0, value=0}, <at 0.2 1>, <at 0.4 0.95>, {tangentMode=RCTM_User, time=1, value=0}]}}})))
  ]
  rendererClasses [&NiagaraMeshRendererProperties]
  renderers [{rendererClass=&NiagaraMeshRendererProperties, rendererIndex=0}]
 }
 Refraction_Glow {
  emitterUpdateScript [
   &EmitterState()
   &SpawnBurst_Instantaneous(Use Spawn Probability=false, UseLoopCountLimit=false)
  ]
  particleSpawnScript [
   &InitializeParticle(Lifetime=0.3, Mesh Scale Mode=::NewEnumerator3, Mesh Uniform Scale=0.04, UsePositionOffset=false, Write Material Random=true) {module=InitializeParticle001}
   &InitialMeshOrientation(
    Rotation=&RandomRangeVector(Minimum=<0 0 0>, Maximum=<1 1 1>)
    Use Orientation Vector=true
    Use Rotation Vector=true
   ) {module=InitialMeshOrientation001}
   &SystemLocation(MaskSpawnGroup=false)
   &CameraOffset(Camera Offset Amount=&MakeCustomFloatFromBool(Bool=$User.bIsFrontEndPreview, True Float=1, False Float=30))
  ]
  particleUpdateScript [
   &ParticleState()
   &DynamicMaterialParameters(
    Index 0 Param 1=&FloatFromCurve(FloatCurve=DataInterface() {propertyValues {Curve {keys [<at 0 1.33>,<at 0.4 1>,<at 1 0.968>]}}})
    Index 0 Param 2=1
    Index 0 Param 3=1
    Index 0 Param 4=1
    Module.Param0WriteEnabled=true
    Module.Param1WriteEnabled=true
    Module.Param2WriteEnabled=true
    Module.Param3WriteEnabled=true
   )
   &ScaleColor(
    Scale Alpha=&FloatFromCurve(FloatCurve=DataInterface() {propertyValues {Curve {keys [
         {interpMode=RCIM_Cubic, tangentMode=RCTM_User, time=0, value=0}
         {interpMode=RCIM_Cubic, tangentMode=RCTM_User, time=0.2, value=1.05}
         {tangentMode=RCTM_User, time=1, value=0}
        ]}}})
    Scale RGB=&MakeVectorFromLinearColorRGB(Color=&ColorFromCurve(ColorCurve=DataInterface() {propertyValues {
        AlphaCurve {keys [<at 0 1 RCIM_Cubic>]}
        BlueCurve {keys [<at 0 2.5726 RCIM_Cubic>,<at 0.6829 1 RCIM_Cubic>]}
        GreenCurve {keys [<at 0 5 RCIM_Cubic>,<at 0.6829 0.4176 RCIM_Cubic>]}
        RedCurve {keys [<at 0 2.2812 RCIM_Cubic>,<at 0.6829 0 RCIM_Cubic>]}
       }}))
    ScaleA=true
    ScaleRGB=true
   )
   &ScaleMeshSize(Scale Factor=&VectorFromFloat(Value=&FloatFromCurve(FloatCurve=DataInterface() {propertyValues {Curve {keys [{tangentMode=RCTM_User, time=0, value=1}, {tangentMode=RCTM_User, time=1, value=0}]}}})))
  ]
  rendererClasses [&NiagaraMeshRendererProperties]
  renderers [{rendererClass=&NiagaraMeshRendererProperties, rendererIndex=0}]
 }
 Wave {
  emitterUpdateScript [
   &EmitterState(Life Cycle Mode=::NewEnumerator1, Loop Behavior=::NewEnumerator1, Loop Delay=0.05, Loop Duration=1, UseLoopDelay=true)
   &SpawnBurst_Instantaneous(Spawn Count=1, Use Spawn Probability=false)
  ]
  particleSpawnScript [
   &InitializeParticle(
    Color=<rgba 1 1 1 4>
    Lifetime=0.35
    Mesh Scale Mode=::NewEnumerator3
    Mesh Uniform Scale=0.2
    UsePositionOffset=false
    Write Material Random=true
   )
   &InitialMeshOrientation(Orientation Axis=<0 0 1>, Orientation Vector=<0 0 1>, Rotation=<0 0.25 0>, Use Orientation Vector=true, Use Rotation Vector=true)
   &InitialMeshOrientation(
    Orientation Axis=<0 0 0>
    Orientation Vector=<0 0 0>
    Rotation=&RandomRangeVector(Minimum=<0 0 0>, Maximum=<1 0 0>)
    Use Orientation Vector=false
    Use Rotation Vector=true
   ) {module=InitialMeshOrientation001}
   &CameraOffset(Camera Offset Amount=&MakeCustomFloatFromBool(Bool=$User.bIsFrontEndPreview, True Float=1, False Float=30))
  ]
  particleUpdateScript [
   &ParticleState()
   &ScaleColor(
    Linear Color Curve=DataInterface() {propertyValues {
      AlphaCurve {keys [<at 0 1>,<at 1 0>]}
      BlueCurve {keys [<at 0.0089 1>,<at 0.4813 1>,<at 0.8577 0.1801>]}
      GreenCurve {keys [<at 0.0089 0>,<at 0.4813 0.7483>,<at 0.8577 0.552>]}
      RedCurve {keys [<at 0.0089 0.4528>,<at 0.4813 0.027>,<at 0.8577 1>]}
     }}
    Scale Mode=::NewEnumerator2
   ) {module=ScaleColor002}
   &DynamicMaterialParameters(
    Index 0 Param 1=&FloatFromCurve(FloatCurve=DataInterface() {propertyValues {Curve {keys [{arriveTangent=0.6769, time=0, value=0.3}, <at 1 1>]}}})
    Index 0 Param 2=&FloatFromCurve(FloatCurve=DataInterface() {propertyValues {Curve {keys [{arriveTangent=-0.5818, interpMode=RCIM_Cubic, leaveTangent=-0.5818, tangentMode=RCTM_User, time=0, value=1}, <at 1 0.8>]}}})
    Index 0 Param 3=&FloatFromCurve(FloatCurve=DataInterface() {propertyValues {Curve {keys [<at 0 0>,<at 0.15 1>,<at 1 1.2>]}}})
    Module.Param0WriteEnabled=true
    Module.Param1WriteEnabled=true
    Module.Param2WriteEnabled=true
    Module.Param3WriteEnabled=false
   )
   &ScaleColor(
    Scale Alpha=&Multiply_Float(A=2, B=&FNTimeOfDayScalar(Night Scale=1, Morning Scale=0.5, Evening Scale=0.5))
    Scale RGB=&VectorFromFloat(Value=&FloatFromCurve(FloatCurve=DataInterface() {propertyValues {Curve {keys [<at 0 70>,<at 0.2 3>,<at 1 1>]}}}))
    ScaleA=true
    ScaleRGB=true
   ) {module=ScaleColor001}
   &ScaleMeshSize(Scale Factor=&VectorFromFloat(Value=&FloatFromCurve(FloatCurve=DataInterface() {propertyValues {Curve {keys [<at 0 0.5>,<at 0.15 1>,<at 1 1.2>]}}})))
  ]
  rendererClasses [&NiagaraMeshRendererProperties]
  renderers [{rendererClass=&NiagaraMeshRendererProperties, rendererIndex=0}]
 }
 Wisps {
  emitterUpdateScript [
   &EmitterState()
   &SpawnBurst_Instantaneous(Use Spawn Probability=false, UseLoopCountLimit=false)
  ]
  particleSpawnScript [
   &InitializeParticle(
    Color Mode=::NewEnumerator0
    Lifetime=0.65
    Sprite Rotation Angle=&RandomRangeFloat(Minimum=15, Maximum=-15)
    Sprite Rotation Mode=::NewEnumerator1
    Sprite Size Mode=::NewEnumerator3
    Uniform Sprite Size=50
    UsePositionOffset=false
    Write Material Random=true
   )
   &DynamicMaterialParameters(Module.Param0WriteEnabled=true, Module.Param1WriteEnabled=true, Module.Param2WriteEnabled=true, Module.Param3WriteEnabled=true)
  ]
  particleUpdateScript [
   &ParticleState()
   &ScaleColor(
    Scale Alpha=&FloatFromCurve(
     FloatCurve=DataInterface() {propertyValues {Curve {keys [
         {interpMode=RCIM_Cubic, tangentMode=RCTM_User, time=0, value=0}
         {interpMode=RCIM_Cubic, tangentMode=RCTM_User, time=0.4, value=1}
         {tangentMode=RCTM_User, time=1, value=0}
        ]}}}
     Scale Curve=0.2
    )
    ScaleA=true
    ScaleRGB=true
   )
   &ScaleSpriteSize(Uniform Curve Sprite Scale=DataInterface() {propertyValues {Curve {keys [
        {arriveTangent=1.9749, interpMode=RCIM_Cubic, leaveTangent=1.9749, tangentMode=RCTM_User, time=0, value=0.3}
        {interpMode=RCIM_Cubic, tangentMode=RCTM_User, time=1, value=1}
       ]}}})
  ]
  rendererClasses [&NiagaraSpriteRendererProperties]
  renderers [{rendererClass=&NiagaraSpriteRendererProperties}]
 }
}
summary {
 systemName=NS_GleamEmote_Portal
 emitters [
  {
   emitterName=Evaporation_Flash
   rendererClasses [&NiagaraSpriteRendererProperties, &NiagaraComponentRendererProperties]
  }
  {emitterName=Refraction_Glow, rendererClasses [&NiagaraMeshRendererProperties]}
  {emitterName=Fragments, rendererClasses [&NiagaraMeshRendererProperties]}
  {emitterName=Wave, rendererClasses [&NiagaraMeshRendererProperties]}
  {emitterName=Wisps, rendererClasses [&NiagaraSpriteRendererProperties]}
 ]
 userVariables [
  User.bIsFrontEndPreview {type=NiagaraBool}
  User.Glowsize {type=NiagaraFloat, defaultValue=0, description=""}
 ]
}
