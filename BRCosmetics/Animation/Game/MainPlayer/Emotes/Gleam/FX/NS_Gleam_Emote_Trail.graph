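# Niagara trail effect for the Gleam emote: a glowing, likely butterfly, flare shedding sparks and wisps.
# Flare keeps a burst-spawned sprite and component-rendered light, both pulsing on sine waves of emitter age.
# Butterfly_Sparks drift down with curl noise and shrink, while Wisps spawn over time and distance, partly following source movement with decaying influence.
# User.Glowsize scales flare, light and wisps but not sparks, and defaults to zero.
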NiagaraSystem /BRCosmetics/Animation/Game/MainPlayer/Emotes/Gleam/FX/NS_Gleam_Emote_Trail
dependencies {usedDataInterfaces [&NiagaraDataInterfaceCurve], usedDynamicInputs [], usedModules [&SystemState]}
emitters {
 Butterfly_Sparks {
  emitterUpdateScript [
   &EmitterState()
   &SpawnRate(SpawnRate=22, Use Spawn Probability=false)
  ]
  particleSpawnScript [
   &InitializeParticle(
    AdjustAlpha=false
    AdjustHue=true
    AdjustSaturation=false
    AdjustValue=false
    Color=<rgba 0.2396 0.658 1 1>
    Color Mode=::NewEnumerator3
    Lifetime Max=1.5
    Lifetime Min=0.5
    Lifetime Mode=::NewEnumerator1
    Sprite Size Mode=::NewEnumerator1
    Uniform Sprite Size Max=6
    Uniform Sprite Size Min=2
    UsePositionOffset=false
    Write Material Random=true
   )
   &ApplyInitialForces()
   &SphereLocation(
    MaskSpawnGroup=false
    Override Seed=false
    Sphere Radius=4
    Surface Only=false
    UseNonUniformScale=false
    UseOffset=false
    UseSphereOrientation=false
   )
  ]
  particleUpdateScript [
   &ParticleState()
   &CurlNoiseForce(Noise Strength=100, Pan Noise=false)
   &Drag(Drag=3, Use Linear Drag=true, Use Rotational Drag=false)
   &GravityForce(Gravity=<0 0 -75>)
   &SolveForcesAndVelocity(Clamp Velocity=false, Limit Acceleration=false)
   &ScaleSpriteSize(Uniform Curve Sprite Scale=DataInterface() {propertyValues {Curve {keys [{interpMode=RCIM_Cubic, tangentMode=RCTM_User, time=0.3, value=1}, {interpMode=RCIM_Cubic, tangentMode=RCTM_User, time=1, value=0}]}}})
   &ScaleColor(ScaleA=true, ScaleRGB=false)
  ]
  rendererClasses [&NiagaraSpriteRendererProperties]
  renderers [{rendererClass=&NiagaraSpriteRendererProperties}]
 }
 Flare {
  emitterUpdateScript [
   &EmitterState(Life Cycle Mode=::NewEnumerator1, Loop Behavior=::NewEnumerator1, Loop Duration Mode=::NewEnumerator1, UseLoopDelay=false)
   &SpawnBurst_Instantaneous(Use Spawn Probability=false, UseLoopCountLimit=false)
   /SetVariables(Emitter.CameraFade=15, Emitter.Depthfade=100) {isSetParameters=true}
  ]
  particleSpawnScript [
   &InitializeParticle(
    Color=<rgba 0.3749 0.7544 1 0.25>
    Position Offset=<0 -0.5 0>
    Position Offset Coordinate Space=::Local
    Sprite Size Mode=::NewEnumerator3
    Uniform Sprite Size=45
    UsePositionOffset=true
    Write Material Random=true
   )
   &DynamicMaterialParameters(
    Index 0 Param 1=0.7
    Index 0 Param 2=1
    Index 0 Param 3=1
    Index 0 Param 4=1
    Module.Param0WriteEnabled=true
    Module.Param1WriteEnabled=true
    Module.Param2WriteEnabled=true
    Module.Param3WriteEnabled=true
   )
   &CameraOffset(Camera Offset Amount=25)
  ]
  particleUpdateScript [
   &ParticleState(Kill Particles When Lifetime Has Elapsed=false, Let Infinitely Lived Particles Die When Emitter Deactivates=true)
   &ScaleSpriteSize(
    Scale Sprite Size Mode=::NewEnumerator0
    Uniform Scale Factor=&Multiply_Float(
     A=&Sine(
      Normalized Angle=$Emitter.Age
      Period=&Waveform("Waveform [1]"=::NewEnumerator8, "[1] Phase Scale"=0.2, "[1] Frequency"=0.2, "[1] Amplitude Min/Max"=<0 1>, "[1] Scale"=0.2)
      Scale=0.02
      Bias=1
     )
     B=$User.Glowsize
    )
   )
   &FN_Gen5_Light(
    Attenutation=45
    Color:LinearColor=$Particles.Color
    Exponent=4
    Intensity:NiagaraFloat=&Multiply_Float(
     A=&Multiply_Float(A=4, B=$User.Glowsize)
     B=&Sine(Normalized Angle=$Emitter.Age, Period=0.2, Scale=0.1, Bias=1)
    )
   )
  ]
  rendererClasses [&NiagaraSpriteRendererProperties, &NiagaraComponentRendererProperties]
  renderers [
   {rendererClass=&NiagaraSpriteRendererProperties}
   {rendererClass=&NiagaraComponentRendererProperties, rendererIndex=1}
  ]
 }
 Wisps {
  emitterUpdateScript [
   &EmitterState(Life Cycle Mode=::NewEnumerator1, Loop Behavior=::NewEnumerator1, Loop Duration=8.5, UseLoopDelay=false)
   &SpawnRate(SpawnRate=2, Use Spawn Probability=false)
   &SpawnPerUnit(Module.Use Max Movement Threshold=true, Module.Use Movement Tolerance=false, Use Spawn Probability=false)
  ]
  particleSpawnScript [
   &InitializeParticle(
    Color Mode=::NewEnumerator0
    Lifetime=0.6
    Sprite Rotation Mode=::NewEnumerator2
    Sprite Size Mode=::NewEnumerator1
    Uniform Sprite Size Max=40
    Uniform Sprite Size Min=35
    UsePositionOffset=false
    Write Material Random=true
   )
   &DynamicMaterialParameters(Module.Param0WriteEnabled=true, Module.Param1WriteEnabled=true, Module.Param2WriteEnabled=true, Module.Param3WriteEnabled=true)
   &ShapeLocation(MaskSpawnGroup=false, Override Seed=false, Sphere Radius=5, UseOwnerScale=true)
  ]
  particleUpdateScript [
   &ParticleState()
   &ScaleColor(
    Scale Alpha=&FloatFromCurve(
     FloatCurve=DataInterface() {propertyValues {Curve {keys [
         {interpMode=RCIM_Cubic, tangentMode=RCTM_User, time=0, value=0}
         {interpMode=RCIM_Cubic, tangentMode=RCTM_User, time=0.4, value=1}
         {tangentMode=RCTM_User, time=1, value=0}
        ]}}}
     Scale Curve=&Multiply_Float(A=0.15, B=$User.Glowsize)
    )
    ScaleA=true
    ScaleRGB=true
   )
   &ScaleSpriteSize(
    Uniform Curve Scale=&Multiply_Float(A=1, B=$User.Glowsize)
    Uniform Curve Sprite Scale=DataInterface() {propertyValues {Curve {keys [
        {arriveTangent=1.3408, interpMode=RCIM_Cubic, leaveTangent=1.3408, tangentMode=RCTM_User, time=0, value=0.3}
        {interpMode=RCIM_Cubic, tangentMode=RCTM_User, time=1, value=1}
       ]}}}
   )
   &InheritSourceMovement(
    Applied Position Scale=&Multiply_VectorByFloat(
     Vector=<0.8 0.8 0.8>
     Float=&FloatFromCurve(FloatCurve=DataInterface() {propertyValues {Curve {keys [
          {arriveTangent=-2.8145, interpMode=RCIM_Cubic, leaveTangent=-2.8145, tangentMode=RCTM_User, time=0, value=1}
          {interpMode=RCIM_Cubic, tangentMode=RCTM_User, time=1, value=0}
         ]}}})
    )
    ApplyForce=false
    ApplyPosition=true
    ApplyPositionFalloff=false
   )
   &CurlNoiseForce(Noise Frequency=25, Noise Strength=50, Pan Noise=false)
   &Drag(Use Linear Drag=true, Use Rotational Drag=false)
   &SolveForcesAndVelocity(Clamp Velocity=false, Limit Acceleration=false)
  ]
  rendererClasses [&NiagaraSpriteRendererProperties]
  renderers [{rendererClass=&NiagaraSpriteRendererProperties}]
 }
}
summary {
 systemName=NS_Gleam_Emote_Trail
 emitters [
  {emitterName=Flare, rendererClasses [&NiagaraSpriteRendererProperties, &NiagaraComponentRendererProperties]}
  {emitterName=Butterfly_Sparks, rendererClasses [&NiagaraSpriteRendererProperties]}
  {emitterName=Wisps, rendererClasses [&NiagaraSpriteRendererProperties]}
 ]
 userVariables [
  User.bIsFrontEndPreview {type=NiagaraBool}
  User.Glowsize {type=NiagaraFloat, defaultValue=0, description=""}
 ]
}
